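# A UI-domain material for ContentOnDemand text that blends the font color with a shimmer color.
# It linearly interpolates between the TextColor and ShimmerColor parameters, using only their RGB channels.
# The blend alpha comes from the shared ShineSwipe material function, fed by the default texture coordinates.
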Material /Game/UI/ContentOnDemand/M_UI_ContentOnDemand_FontMaterial
graph {
 properties {bUsedWithStaticMesh=false, materialDomain=MD_UI, pixelDepthOffsetMode=PDOM_Legacy, usedShadingModels=Default Lit}
 nodes [LinearInterpolate(
   A=VectorParameter() {parameterName=TextColor, defaultValue=<rgba 1 1 1 1>}.RGB
   B=VectorParameter() {parameterName=ShimmerColor, defaultValue=<rgba 0.5184 0.6463 0.7656 0.2696>}.RGB
   Alpha=MaterialFunctionCall!(In_Uv=TextureCoordinate()) {materialFunction=/Game/UI/Foundation/Materials/M-Function-ShineSwipe}.Result
  ) {constB=0.5}]
}
